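# Translucent unlit Niagara sprite material that draws a pixelated heart for the Nox pickaxe effects.
# A Pixalize parameter snaps the UVs to a grid before sampling the heart texture, with derivatives from the unsnapped UVs keeping mip selection stable.
# The red channel is thresholded into a hard mask and multiplied by particle alpha.
# Particle RGB feeds MF_FX_EmissiveAdjustments for the glow color.
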Material /BRCosmetics/Weapons/FORT_Melee/Pickaxe_Nox/Materials/M_Heart
graph {
 properties {bUsedWithNiagaraSprites=true, blendMode=BLEND_Translucent, pixelDepthOffsetMode=PDOM_Legacy, shadingModel=MSM_Unlit, usedShadingModels=Unlit}
 nodes [
  n1 = ParticleColor()
  MaterialFunctionCall(Emissive=@n1.RGB) {materialFunction=&MF_FX_EmissiveAdjustments}
  n2 = ScalarParameter() {defaultValue=25, parameterName=Pixalize, sliderMax=200}
  n3 = Multiply(A=TextureCoordinate(), B=@n2) {constB=100}
  n4 = TextureCoordinate()
  n5 = TextureSampleParameter2D!(UVs=Divide(A=Round(None=@n3), B=@n2), "DDX(UVs)"=DDX(Value=@n4), "DDY(UVs)"=DDY(Value=@n4), Apply View MipBias) {mipValueMode=TMVM_Derivative, parameterName=Texture, texture=/BRCosmetics/UI/Foundation/Textures/Banner/Standard/T-Banners-Icons-Heart}
  Multiply(A=@n1.A, B=If(A=@n5.R, "A > B"=Constant() {r=1}, "A < B"=Constant()) {constB=0.1})
 ]
}
