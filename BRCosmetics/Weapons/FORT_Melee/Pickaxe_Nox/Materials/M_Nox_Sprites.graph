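# Unlit translucent Niagara sprite material for the Nox pickaxe effects, drawing concentric rings.
# Distance from an offset center, with optional UV pixelization through a static switch, is scaled and shifted by the dynamic Progress parameter to create repeating bands.
# The bands alternate two colors through an emissive adjustment function.
# A sphere mask on each band shapes ring thickness, multiplied by a circular cutout and particle alpha.
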Material /BRCosmetics/Weapons/FORT_Melee/Pickaxe_Nox/Materials/M_Nox_Sprites
graph {
 properties {bUsedWithNiagaraSprites=true, blendMode=BLEND_Translucent, pixelDepthOffsetMode=PDOM_Legacy, shadingModel=MSM_Unlit, usedShadingModels=Unlit}
 nodes [
  n1 = ScalarParameter() {defaultValue=128, parameterName=Pixalize, sliderMax=200}
  n2 = Multiply(A=TextureCoordinate(), B=@n1) {constB=100}
  n3 = StaticSwitchParameter(True=Divide(A=Round(None=@n2), B=@n1), False=TextureCoordinate()) {parameterName=Pixalize}
  n4 = ScalarParameter() {defaultValue=0.5, parameterName=Offset X, sliderMax=1}
  n5 = ScalarParameter() {defaultValue=0.5, parameterName=Offset Y, sliderMax=1}
  n6 = Frac(None=Add(A=@n3, B=AppendVector(A=@n4, B=@n5)) {constB=0.5})
  n7 = Constant2Vector() {g=0.5, r=0.5}
  n8 = Add(A=Multiply(A=Distance(A=@n6, B=@n7)) {constB=3}, B=DynamicParameter() {paramNames [Progress,Param2,Param3,Param4]}.Progress)
  n9 = Fmod(A=@n8, B=Constant() {r=2})
  MaterialFunctionCall(Emissive=LinearInterpolate(
    A=VectorParameter() {parameterName=Color 1, defaultValue=<rgba 0 0 0 1>}.RGB
    B=VectorParameter() {parameterName=Color 2, defaultValue=<rgba 0 0 0 1>}.RGB
    Alpha=Floor(None=@n9)
   )) {materialFunction=&MF_FX_EmissiveAdjustments}
  n10 = Frac(None=@n8)
  n11 = ScalarParameter() {parameterName=Radius}
  n12 = ScalarParameter() {defaultValue=0.1, parameterName=Thickness, sliderMax=1}
  n13 = ScalarParameter() {parameterName=Hardness}
  n14 = Frac(None=Add(A=@n3, B=Constant2Vector() {g=0.5, r=0.5}) {constB=0.5})
  n15 = Constant2Vector() {g=0.5, r=0.5}
  n16 = SphereMask(A=Frac(None=Distance(A=@n14, B=@n15)), B=Constant(), Radius=ScalarParameter() {defaultValue=0.5, parameterName=Cutout Radius, sliderMax=1}) {attenuationRadius=0.501}
  Multiply(A=Multiply(A=SphereMask(A=@n10, B=@n11, Radius=@n12, Hardness=@n13) {attenuationRadius=0.1}, B=OneMinus(None=@n16)), B=ParticleColor().A)
  ScalarParameter()
 ]
}
