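# Unlit, two-sided translucent effect material for the Nox pickaxe, flagged for Niagara ribbons and mesh particles, that draws diagonal bands on a tube.
# Panned UVs are summed, wrapped and sphere-masked into stripes, which smoothstep between two colors for emissive.
# Opacity math combines particle alpha, view-angle fade, edge masks and subtracted scrolling noise for a soft leading edge.
# A parabolic V-coordinate term times vertex normal computes a center bulge offset.
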Material /BRCosmetics/Weapons/FORT_Melee/Pickaxe_Nox/Materials/M_Pickaxe_Nox_DiagonalTube
graph {
 properties {
  bUsedWithNiagaraMeshParticles=true
  bUsedWithNiagaraRibbons=true
  blendMode=BLEND_Translucent
  opacityMaskClipValue=1
  pixelDepthOffsetMode=PDOM_Legacy
  shadingModel=MSM_Unlit
  twoSided=true
  usedShadingModels=Unlit
 }
 nodes [
  n1 = Panner(Coordinate=TextureCoordinate() {coordinateIndex=1}) {speedX=-1}
  n2 = Multiply(A=ComponentMask(None=@n1) {r=true}, B=ScalarParameter() {defaultValue=1, parameterName=Tile X}) {constB=3}
  n3 = Multiply(A=ComponentMask(None=@n1) {g=true}, B=ScalarParameter() {defaultValue=1, parameterName=Tile Y}) {constB=3}
  n4 = SphereMask(
   A=Frac(None=Add(A=@n2, B=@n3))
   B=Constant() {r=0.5}
   Radius=ScalarParameter() {defaultValue=0.1, parameterName=Thickness, sliderMax=1}
   Hardness=ScalarParameter() {parameterName=Hardness, sliderMax=1}
  ) {attenuationRadius=0.1, hardnessPercent=0}
  n5 = TextureCoordinate()
  n6 = OneMinus(None=ComponentMask(None=@n5) {r=true})
  n7 = ScalarParameter() {defaultValue=1, parameterName=Tile X}
  n8 = ComponentMask(None=@n5) {g=true}
  n9 = ScalarParameter() {defaultValue=1, parameterName=Tile Y}
  n10 = Frac(None=Add(A=Multiply(A=@n6, B=@n7) {constB=3}, B=Multiply(A=@n8, B=@n9) {constB=3}))
  n11 = Constant() {r=0.5}
  n12 = ScalarParameter() {defaultValue=0.1, parameterName=Thickness, sliderMax=1}
  n13 = ScalarParameter() {parameterName=Hardness, sliderMax=1}
  Add(A=@n4, B=Multiply(A=SphereMask(A=@n10, B=@n11, Radius=@n12, Hardness=@n13) {attenuationRadius=0.1, hardnessPercent=0}) {constB=0.5})
  n14 = ScalarParameter() {parameterName=SmoothStepMin}
  n15 = ScalarParameter() {defaultValue=3, parameterName=SmoothStepMax, sliderMax=5}
  n16 = Add(B=@n4)
  MaterialFunctionCall(Emissive=LinearInterpolate(
    A=VectorParameter() {parameterName=Color 1, defaultValue=<rgba 1 0.1506 0.6248 1>}.RGB
    B=VectorParameter() {parameterName=Color 2, defaultValue=<rgba 1 0.8905 0 1>}.RGB
    Alpha=SmoothStep(Min=@n14, Max=@n15, Value=@n16) {constMax=3, constMin=-0.5}
   )) {materialFunction=&MF_FX_EmissiveAdjustments}
  n17 = Multiply(A=ParticleColor().A, B=ScalarParameter() {defaultValue=1, parameterName=Opacity})
  n18 = MaterialFunctionCall("FadeSensitivity (optional)"=ScalarParameter() {defaultValue=3, parameterName=View Angle Fade Sensitivity, sliderMax=16}) {materialFunction=&FadeBasedOnViewAngle}
  n19 = ComponentMask(None=TextureCoordinate()) {g=true}
  n20 = Constant() {r=0.5}
  n21 = ComponentMask(None=TextureCoordinate()) {r=true}
  n22 = Multiply(
   A=Multiply(A=Multiply(A=@n17, B=@n18.Result), B=SphereMask(A=@n19, B=@n20) {attenuationRadius=0.5, hardnessPercent=0})
   B=Power(Base=OneMinus(None=@n21))
  )
  n23 = TextureCoordinate() {coordinateIndex=1}
  n24 = VectorParameter() {group=Soft Leading Edge, parameterName=Soft Leading Edge UV Scale, sortPriority=6, defaultValue=<rgba 0.2 1 0 0>}
  n25 = ComponentMask(None=@n24.RGB) {g=true, r=true}
  n26 = VectorParameter() {group=Soft Leading Edge, parameterName=Soft Leading Edge Panner Speed, sortPriority=7, defaultValue=<rgba -0.65 0 0 0>}
  n27 = TextureSampleParameter2D(UVs=Panner(Coordinate=Multiply(A=@n23, B=@n25), Speed=ComponentMask(None=@n26.RGB) {g=true, r=true}) {bFractionalPart=true}) {
   group=Soft Leading Edge
   parameterName=Soft Leading Edge Texture
   samplerSource=SSM_Wrap_WorldGroupSettings
   sortPriority=4
   texture=&T_NoisePack_01
  }
  n28 = Multiply(
   A=Add(A=@n27.R, B=ScalarParameter() {group=Soft Leading Edge, parameterName=Bias, sortPriority=3})
   B=ScalarParameter() {defaultValue=0.5, group=Soft Leading Edge, parameterName=Scale, sortPriority=2}
  )
  n29 = ComponentMask(None=TextureCoordinate()) {r=true}
  n30 = ScalarParameter() {group="Fade Leading && Trailing Edges", parameterName=Leading Edge Fade Bias, sortPriority=2}
  n31 = Saturate(None=Multiply(A=Add(A=@n29, B=@n30), B=ScalarParameter() {defaultValue=3, group="Fade Leading && Trailing Edges", parameterName=Leading Edge Fade Scale}) {constB=3})
  n32 = ScalarParameter() {defaultValue=1, group=Alpha, parameterName=Fade Leading Edge, sortPriority=4}
  n33 = OneMinus(None=@n29)
  n34 = ScalarParameter() {group="Fade Leading && Trailing Edges", parameterName=Trailing Edge Fade Bias, sortPriority=3}
  n35 = Saturate(None=Multiply(
    A=Add(A=@n33, B=@n34)
    B=ScalarParameter() {defaultValue=2, group="Fade Leading && Trailing Edges", parameterName=Trailing Edge Fade Scale, sortPriority=1}
   ) {constB=2})
  n36 = ScalarParameter() {defaultValue=1, group=Alpha, parameterName=Fade Trailing Edge, sortPriority=5}
  Multiply(
   A=Subtract(A=Multiply(A=@n22) {constB=4}, B=@n28)
   B=Multiply(A=LinearInterpolate(B=@n31, Alpha=@n32) {constA=1}, B=LinearInterpolate(B=@n35, Alpha=@n36) {constA=1})
  )
  n37 = ConstantBiasScale(None=ComponentMask(None=TextureCoordinate()) {g=true}) {bias=-0.5, scale=2}
  n38 = DotProduct(A=@n37, B=@n37)
  Multiply(A=Multiply(A=OneMinus(None=@n38), B=ScalarParameter() {defaultValue=6, parameterName=Center Bulge}), B=VertexNormalWS())
  n39 = ComponentMask(None=TextureCoordinate()) {r=true}
  Power(
   Base=Add(A=@n28, B=OneMinus(None=@n39))
   Exp=ScalarParameter() {defaultValue=2, group=Soft Leading Edge, parameterName=Soft Leading Edge Contrast, sortPriority=1}
  ) {constExponent=6}
 ]
}
