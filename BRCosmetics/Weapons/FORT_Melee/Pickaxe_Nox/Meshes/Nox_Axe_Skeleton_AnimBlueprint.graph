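# Animation Blueprint for the Nox pickaxe skeleton that continuously spins its dyn_ring bone, in opposite directions for main and offhand weapons.
# On initialization it flags offhand use by checking whether the mesh component's display name contains "Offhand".
# Each update derives a yaw from game time and that flag's sign.
# The AnimGraph applies it additively in bone space over the reference pose.
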AnimBlueprint /BRCosmetics/Weapons/FORT_Melee/Pickaxe_Nox/Meshes/Nox_Axe_Skeleton_AnimBlueprint
target_skeleton=&/Nox_Axe_Skeleton
graphs [
 AnimGraph {nodes [
   n1 = "Transform (Modify) Bone - Bone: dyn_ring"(
    ComponentPose=Local To Component(LocalPose=Local Space Ref Pose())
    Alpha=1
    Translation="0,0,0"
    Rotation=Get RingRotation() -> (RingRotation:Rotator)
    Scale="1,1,1"
   ) {
    rotationMode=BMM_Additive
    rotationSpace=BCS_BoneSpace
    translationMode=BMM_Additive
    translationSpace=BCS_BoneSpace
    alphaBoolBlend {blendOption=Linear}
    boneToModify {boneName=dyn_ring}
   }
   Output Pose(Result=Component To Local(ComponentPose=@n1)) {layerGroup=None}
  ]}
 EventGraph {nodes [
   Event BlueprintUpdateAnimation()
   n1 = SelectFloat(A=1, B=-1, bPickA=Get offsetWeapon() -> (offsetWeapon:bool))
   n2 = GetGameTimeInSeconds()
   > Set RingRotation(RingRotation:Rotator=MakeRotator(Yaw=float * float!(A=@n2, B=20, C:double=@n1) -> (ReturnValue) {is=title K2Node_PromotableOperator})) -> (Output_Get:Rotator)
   Event BlueprintInitializeAnimation()
   > n3 = Set MeshName(MeshName:Skeletal Mesh Component Object Reference=GetPartSkeletalMeshComponent()) -> (Output_Get:Skeletal Mesh Component Object Reference)
   n4 = GetDisplayName(Object=@n3)
   > n5 = Branch(Condition=Contains(SearchIn=@n4, Substring=Offhand))
   > Set offsetWeapon(offsetWeapon:bool=true) -> (Output_Get:bool)
   Set offsetWeapon(execute=@n5.else, offsetWeapon=false)
   "Equal Exactly (String)"(A=@n4, B=WeaponMeshOffhand)
  ]}
]
variables [
 Delta Time X {type=number}
 MeshName {type=SkeletalMeshComponent}
 offsetWeapon {type=boolean}
 RingRotation {type=Rotator}
 UpDownMotion {type=number}
]
